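# Post-process material for the turntable capture assistant that renders the metallic channel of the scene.
# An If compares scene depth to a far-distance threshold, passing through GBuffer Metallic for nearer pixels.
# Pixels at or beyond the threshold take the background parameter from the Collection_TurntableCaptures parameter collection, separating empty backdrop from captured geometry.
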Material /Game/FN_PublicTurntable/TurntableCaptureAssistant/Materials/M_TurntableCapture_Metallic
graph {
 properties {blendableLocation=BL_SceneColorAfterDOF, materialDomain=MD_PostProcess, pixelDepthOffsetMode=PDOM_Legacy, usedShadingModels=Default Lit}
 nodes [
  n1 = CollectionParameter() {collection=&/Collection_TurntableCaptures, parameterName=background}
  If(
   A=ComponentMask(None=SceneTexture() {sceneTextureId=PPI_SceneDepth}.Color) {r=true}
   B=Constant() {r=65535}
   "A > B"=@n1
   "A == B"=@n1
   "A < B"=ComponentMask(None=SceneTexture() {sceneTextureId=PPI_Metallic}.Color) {r=true}
  )
 ]
}
