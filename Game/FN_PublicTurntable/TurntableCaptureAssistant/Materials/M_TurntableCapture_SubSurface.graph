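# Post-process material for turntable captures that replaces the scene image with a view of the GBuffer subsurface color.
# Subsurface color is raised to a Power parameter, normalized and multiplied by the opacity buffer.
# A desaturated copy gates the result, so pixels without positive subsurface color turn black.
# A scene depth test fills far, empty pixels with the background color from the Collection_TurntableCaptures parameter collection.
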Material /Game/FN_PublicTurntable/TurntableCaptureAssistant/Materials/M_TurntableCapture_SubSurface
graph {
 properties {blendableLocation=BL_SceneColorAfterDOF, materialDomain=MD_PostProcess, pixelDepthOffsetMode=PDOM_Legacy, usedShadingModels=Default Lit}
 nodes [
  n1 = ComponentMask(None=CollectionParameter() {collection=&/Collection_TurntableCaptures, parameterName=background}) {b=true, g=true, r=true}
  n2 = Power(Base=SceneTexture() {sceneTextureId=PPI_SubsurfaceColor}.Color, Exp=ScalarParameter() {defaultValue=0.35, parameterName=Power})
  n3 = Desaturation(None=@n2)
  n4 = Constant()
  n5 = Multiply(A=Normalize(VectorInput=@n2), B=ComponentMask(None=SceneTexture() {sceneTextureId=PPI_Opacity}.Color) {r=true})
  If(
   A=ComponentMask(None=SceneTexture() {sceneTextureId=PPI_SceneDepth}.Color) {r=true}
   B=Constant() {r=65535}
   "A > B"=@n1
   "A == B"=@n1
   "A < B"=ComponentMask(None=If(A=@n3, B=@n4, "A > B"=@n5, "A == B"=@n4, "A < B"=@n4)) {b=true, g=true, r=true}
  )
 ]
}
