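# Post-process material for turntable captures that visualizes the metallic buffer as a check for in-between values.
# Pixels at maximum scene depth take the background color from the Collection_TurntableCaptures parameter collection.
# On other pixels, nested If nodes compare metallic to the Low and High parameters with a small tolerance.
# Values matching either render black, while all others take the collection's high color.
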Material /Game/FN_PublicTurntable/TurntableCaptureAssistant/Materials/M_TurntableProcess_Metallic
graph {
 properties {blendableLocation=BL_SceneColorAfterDOF, materialDomain=MD_PostProcess, pixelDepthOffsetMode=PDOM_Legacy, usedShadingModels=Default Lit}
 nodes [
  Constant() {r=1}
  n1 = CollectionParameter() {collection=&/Collection_TurntableCaptures, parameterName=background}
  n2 = ComponentMask(None=SceneTexture() {sceneTextureId=PPI_Metallic}.Color) {r=true}
  n3 = CollectionParameter() {collection=&Collection_TurntableCaptures, parameterName=high}
  n4 = If(A=@n2, B=ScalarParameter() {defaultValue=1, parameterName=High}, "A > B"=@n3, "A == B"=Constant(), "A < B"=@n3) {equalsThreshold=0.05}
  If(
   A=ComponentMask(None=SceneTexture() {sceneTextureId=PPI_SceneDepth}.Color) {r=true}
   B=Constant() {r=65535}
   "A > B"=@n1
   "A == B"=@n1
   "A < B"=If(A=@n2, B=ScalarParameter() {parameterName=Low}, "A > B"=@n4, "A == B"=Constant(), "A < B"=@n4) {equalsThreshold=0.05}
  )
 ]
}
