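# Post-process material for the turntable capture assistant that turns the scene into a mask of pixels with subsurface color.
# A scene depth test gives pixels at or beyond a far cutoff the background value from the Collection_TurntableCaptures parameter collection.
# Nearer pixels raise subsurface color to the Power parameter, desaturate it and compare it to the Low threshold.
# Those above Low take the collection's exists value, and the rest become zero.
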Material /Game/FN_PublicTurntable/TurntableCaptureAssistant/Materials/M_TurntableProcess_SubSurface
graph {
 properties {blendableLocation=BL_SceneColorAfterDOF, materialDomain=MD_PostProcess, pixelDepthOffsetMode=PDOM_Legacy, usedShadingModels=Default Lit}
 nodes [
  Constant() {r=1}
  n1 = CollectionParameter() {collection=&/Collection_TurntableCaptures, parameterName=background}
  n2 = Power(Base=SceneTexture() {sceneTextureId=PPI_SubsurfaceColor}.Color, Exp=ScalarParameter() {defaultValue=0.1, parameterName=Power})
  n3 = Constant()
  If(
   A=ComponentMask(None=SceneTexture() {sceneTextureId=PPI_SceneDepth}.Color) {r=true}
   B=Constant() {r=65535}
   "A > B"=@n1
   "A == B"=@n1
   "A < B"=If(
    A=Desaturation(None=@n2)
    B=ScalarParameter() {parameterName=Low}
    "A > B"=CollectionParameter() {collection=&Collection_TurntableCaptures, parameterName=exists}
    "A == B"=@n3
    "A < B"=@n3
   )
  )
 ]
}
