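; Turntable capture helper actor that renders Turntable_AssetBase actors through selectable post-process materials.
; activatePostProcess either weights the chosen material among cached blendables on its PostProcess component, or shows a viewport-sized render target in a widget and captures from the player camera.
; Both capture paths set show-only actors on the SceneCapture2D.
; editor_captureNow renders each textureTargetMap target with its paired material, then resets the component; LEGACY_captureNow duplicates it.
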(fn activatePostProcess
  (params (inCaptureData :type Turntable%20Captures%20Data%20Struct%20Structure))
  (bind (_friendlyname_5_e9ec577a47c0555a0d1a56a2f9141cbc _material_6_eb22c6e14f0560116c76a1b4ee4d2295) (Utilities|Struct|BreakTurntableCapturesDataStruct inCaptureData))
  (bind (_incapturedata_friendlyname_5_e9ec577a47c0555a0d1a56a2f9141cbc _incapturedata_material_6_eb22c6e14f0560116c76a1b4ee4d2295) (Utilities|Struct|BreakTurntableCapturesDataStruct (Variables|GetinCaptureData_FriendlyName_5_E9EC577A47C0555A0D1A56A2F9141CBC)))
  (Utilities|Array|AddUnique (Variables|Core|GetcurrentBlendables) _material_6_eb22c6e14f0560116c76a1b4ee4d2295) ; Cache all instances we use to verify they are off when we override blendables
  (if (Variables|Game|GetusePostProcessForGameCaptures)
    (for _array_element (Variables|Core|GetcurrentBlendables)
      (Rendering|AddorUpdateBlendable (Variables|Default|GetPostProcess) _array_element (Math|Conversions|ToFloat(Boolean) (Utilities|Equal(Object) _array_element _incapturedata_material_6_eb22c6e14f0560116c76a1b4ee4d2295)) :!class /Script/Engine.PostProcessComponent))
    (else
      (Utilities|IsValid (Game|GetCaptureWidget)
        (:Is%20Valid
          (Class|TurntableCapturesRenderDisplayWidget|SetNewTexture (Game|GetCaptureWidget) (Game|GetGameCaptureTarget) (Variables|GetinCaptureData))
          (Camera|MovetoCameraManager)
          (Game|GameCaptureNow (Variables|GetinCaptureData)))))))

(fn editor_captureNow
  (Utilities|FlowControl|Sequence
    (:then_0
      (bind _outactors (Actor|GetAllActorsOfClass "/Game/FN_PublicTurntable/Turntable_AssetBase.Turntable_AssetBase_C"))
      (Variables|/Script/Engine.SceneCaptureComponent|SetShowOnlyActors _outactors (Variables|Default|GetSceneCaptureComponent2D)))
    (:then_1
      (Variables|/Script/Engine.SceneCaptureComponent|SetbCaptureEveryFrame true (Variables|Default|GetSceneCaptureComponent2D))
      (bind _keys (Utilities|Map|Keys (Variables|Editor|GettextureTargetMap)))
      (for (_array_element _array_index) _keys
        (Variables|/Script/Engine.SceneCaptureComponent2D|SetTextureTarget _array_element (Variables|Default|GetSceneCaptureComponent2D))
        (bind _values (Utilities|Map|Values (Variables|Editor|GettextureTargetMap)))
        (Variables|/Script/Engine.SceneCaptureComponent2D|SetPostProcessSettings (Utilities|Struct|MakePostProcessSettings (Utilities|Struct|MakeWeightedBlendables (Utilities|Array|MakeArray (Utilities|Struct|MakeWeightedBlendable 1.0 (.Material_6_EB22C6E14F0560116C76A1B4EE4D2295 (Utilities|Struct|BreakTurntableCapturesDataStruct (Utilities|Array|Get(aref) _values _array_index))))))) (Variables|Default|GetSceneCaptureComponent2D))
        (Rendering|SceneCapture|CaptureScene (Variables|Default|GetSceneCaptureComponent2D) :!class /Script/Engine.SceneCaptureComponent2D))
      (Variables|/Script/Engine.SceneCaptureComponent|SetbCaptureEveryFrame false (Variables|Default|GetSceneCaptureComponent2D))
      (Variables|/Script/Engine.SceneCaptureComponent2D|SetTextureTarget "" (Variables|Default|GetSceneCaptureComponent2D))
      (Variables|/Script/Engine.SceneCaptureComponent2D|SetPostProcessSettings :self (Variables|Default|GetSceneCaptureComponent2D)))))

(fn game_captureNow
  (params (inCaptureData :type Turntable%20Captures%20Data%20Struct%20Structure))
  (bind (_incapturedata_friendlyname_5_e9ec577a47c0555a0d1a56a2f9141cbc _incapturedata_material_6_eb22c6e14f0560116c76a1b4ee4d2295) (Utilities|Struct|BreakTurntableCapturesDataStruct (Variables|GetinCaptureData_FriendlyName_5_E9EC577A47C0555A0D1A56A2F9141CBC)))
  (Transformation|SetActorLocationAndRotation (Camera|GetCameraLocation (Game|GetPlayerCameraManager 0)) (Camera|GetCameraRotation (Game|GetPlayerCameraManager 0)) false true)
  (Utilities|FlowControl|Sequence
    (:then_0
      (bind _outactors (Actor|GetAllActorsOfClass "/Game/FN_PublicTurntable/Turntable_AssetBase.Turntable_AssetBase_C"))
      (Variables|/Script/Engine.SceneCaptureComponent|SetShowOnlyActors _outactors (Variables|Default|GetSceneCaptureComponent2D)))
    (:then_1
      (Variables|/Script/Engine.SceneCaptureComponent|SetbCaptureEveryFrame true (Variables|Default|GetSceneCaptureComponent2D))
      (Variables|/Script/Engine.SceneCaptureComponent2D|SetTextureTarget (Game|GetGameCaptureTarget) (Variables|Default|GetSceneCaptureComponent2D))
      (for _array_element (Variables|Core|GetcurrentBlendables)
        (Rendering|AddorUpdateBlendable (Variables|Default|GetPostProcess) _array_element (Math|Conversions|ToFloat(Boolean) (Utilities|Equal(Object) _array_element _incapturedata_material_6_eb22c6e14f0560116c76a1b4ee4d2295)) :!class /Script/Engine.PostProcessComponent))
      (Rendering|SceneCapture|CaptureScene (Variables|Default|GetSceneCaptureComponent2D) :!class /Script/Engine.SceneCaptureComponent2D))))

(fn getCaptureWidget
  (Utilities|FlowControl|Sequence
    (:then_0
      (Utilities|IsValid (Variables|Game|GetcaptureWidget)
        (:Is%20Not%20Valid
          (bind _returnvalue (UserInterface|CreateWidget "/Game/FN_PublicTurntable/TurntableCaptureAssistant/UI/TurntableCaptures_RenderDisplayWidget.TurntableCaptures_RenderDisplayWidget_C"))
          (bind _output_get (Variables|Game|SetcaptureWidget _returnvalue))
          (UserInterface|Viewport|AddtoViewport _output_get 111111110))))
    (:then_1
      (return (Variables|Game|GetcaptureWidget)))))

(fn getGameCaptureTarget
  (bind (_returnvalue_x _returnvalue_y) (Math|Vector2D|BreakVector2D (Viewport|GetViewportSize)))
  (Utilities|FlowControl|Sequence
    (:then_0
      (Utilities|IsValid (Variables|Game|GetcurrentGameCaptureTarget)
        (:Is%20Not%20Valid
          (bind _returnvalue (Rendering|CreateRenderTarget2D (Math|Float|Truncate _returnvalue_x) (Math|Float|Truncate _returnvalue_y)))
          (Variables|Game|SetcurrentGameCaptureTarget _returnvalue))))
    (:then_1
      (return (Variables|Game|GetcurrentGameCaptureTarget)))))

(fn LEGACY_captureNow
  (Utilities|FlowControl|Sequence
    (:then_0
      (bind _outactors (Actor|GetAllActorsOfClass "/Game/FN_PublicTurntable/Turntable_AssetBase.Turntable_AssetBase_C"))
      (Variables|/Script/Engine.SceneCaptureComponent|SetShowOnlyActors _outactors (Variables|Default|GetSceneCaptureComponent2D)))
    (:then_1
      (Variables|/Script/Engine.SceneCaptureComponent|SetbCaptureEveryFrame true (Variables|Default|GetSceneCaptureComponent2D))
      (bind _keys (Utilities|Map|Keys (Variables|Editor|GettextureTargetMap)))
      (for (_array_element _array_index) _keys
        (Variables|/Script/Engine.SceneCaptureComponent2D|SetTextureTarget _array_element (Variables|Default|GetSceneCaptureComponent2D))
        (bind _values (Utilities|Map|Values (Variables|Editor|GettextureTargetMap)))
        (Variables|/Script/Engine.SceneCaptureComponent2D|SetPostProcessSettings (Utilities|Struct|MakePostProcessSettings (Utilities|Struct|MakeWeightedBlendables (Utilities|Array|MakeArray (Utilities|Struct|MakeWeightedBlendable 1.0 (.Material_6_EB22C6E14F0560116C76A1B4EE4D2295 (Utilities|Struct|BreakTurntableCapturesDataStruct (Utilities|Array|Get(aref) _values _array_index))))))) (Variables|Default|GetSceneCaptureComponent2D))
        (Rendering|SceneCapture|CaptureScene (Variables|Default|GetSceneCaptureComponent2D) :!class /Script/Engine.SceneCaptureComponent2D))
      (Variables|/Script/Engine.SceneCaptureComponent|SetbCaptureEveryFrame false (Variables|Default|GetSceneCaptureComponent2D))
      (Variables|/Script/Engine.SceneCaptureComponent2D|SetTextureTarget "" (Variables|Default|GetSceneCaptureComponent2D))
      (Variables|/Script/Engine.SceneCaptureComponent2D|SetPostProcessSettings :self (Variables|Default|GetSceneCaptureComponent2D)))))

(fn moveToCameraManager
  (Transformation|SetActorLocationAndRotation (Camera|GetCameraLocation (Game|GetPlayerCameraManager 0)) (Camera|GetCameraRotation (Game|GetPlayerCameraManager 0))))

(fn ConstructionScript)
